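# Translucent Niagara mesh-particle material for the BadBear skin's drop effect, built on the character Pre/PostFX functions with an optional Slurp layer.
# Glowing two-color vines combine texture channels warped by scrolling, axis-blended noise, fresnel and camera-relative brightness.
# Time-driven pulses from pre-skinned-position sphere masks boost the vine emissive.
# A masked world position offset calculation pushes geometry away from a bear-heart point.
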Material /BRCosmetics/Athena/Effects/Particles/Skins/BadBear/M_Skin_BadBear_Drop_FX
graph {
 properties {bUsedWithNiagaraMeshParticles=true, blendMode=BLEND_Translucent, pixelDepthOffsetMode=PDOM_Legacy, usedShadingModels=Default Lit}
 nodes [
  n1 = Power(
   Base=DepthFade(FadeDistance=ScalarParameter() {defaultValue=50, parameterName=DepthFade})
   Exp=ScalarParameter() {defaultValue=1.9, parameterName=DepthFadePow}
  )
  LinearInterpolate(A=@n1, B=Multiply(A=@n1, B=ParticleColor().A), Alpha=ScalarParameter() {parameterName=UseParticleAlpha})
  n2 = MaterialFunctionCall() {materialFunction=&BaseFN_CharacterFunction_PreFX}
  n3 = TextureObjectParameter() {
   group=Skin FX
   parameterName=SkinFX_Mask
   samplerType=SAMPLERTYPE_Masks
   texture=/BRCosmetics/Characters/Player/Male/Large/Bodies/M_LRG_BadBear/Textures/T_Character_BadBear_Body_FX
  }
  n4 = NamedRerouteDeclaration(None=TextureSample(Tex=@n3) {samplerType=SAMPLERTYPE_Masks}.RGB) {name=FXMask, nodeColor=<rgba 1 0.7734 0 1>}
  n5 = StaticComponentMaskParameter(None=@n4) {defaultB=true, group=Skin FX, parameterName=Slurp Channel}
  n6 = StaticComponentMaskParameter(None=@n4) {defaultG=true, group=Skin FX, parameterName=Slurp Bubbles Channel}
  n7 = StaticComponentMaskParameter(None=@n4) {defaultB=true, group=Skin FX, parameterName=Slurp Normal Distortion Channel}
  n8 = GetMaterialAttributes(None=StaticSwitchParameter(
    True=MaterialFunctionCall!(Alpha=@n5, AO=@n2.AO, BubblesMask=@n6, SlurpNormalDistortionMask=@n7, Pre Skinned Position=@n2.AO, Material=@n2.AO) {materialFunction=/Game/Weapons/WeaponSkins/MaterialFunction/MF_Slurp}.Result
    False=@n2.Result
   ) {group=Skin FX, parameterName=UseSlurp}) {attributeGetTypes [#1, #2, #3]}
  n9 = NamedRerouteUsage()
  n10 = LinearInterpolate(A=@n8.Emissive Color, Alpha=NamedRerouteUsage()) {constB=0}
  n11 = NamedRerouteUsage()
  n12 = MaterialFunctionCall(
   AO=@n2.AO
   Distance Blend=@n2.AO
   Fresnel=@n2.AO
   Gmap=@n2.AO
   Pre Skinned Local Position=@n2.AO
   Material=SetMaterialAttributes!(MaterialAttributes=@n8.MaterialAttributes, World Position Offset=Add(A=@n8.World Position Offset, B=@n9), Emissive Color=Add(A=@n10, B=@n11)) {attributeSetTypes [#3, #2]}
  ) {materialFunction=&BaseFN_CharacterFunction_PostFX}
  MaterialFunctionCall(In=@n12.Result) {materialFunction=&CINE_MF_CharPostFX}
  MaterialFunctionCall(Emissive=GetMaterialAttributes(None=@n8.MaterialAttributes) {attributeGetTypes [#2]}.Emissive Color) {materialFunction=&MF_FX_EmissiveAdjustments}
  n13 = PreSkinnedNormal()
  NamedRerouteDeclaration(None=ShaderStageSwitch(PixelShader=VertexInterpolator(VS=@n13).PS, VertexShader=@n13)) {name=PreskinnedNor, nodeColor=<rgba 1 0 0.6094 1>}
  n14 = PreSkinnedPosition()
  n15 = VertexInterpolator(VS=@n14)
  n16 = ComponentMask(None=NamedRerouteDeclaration(None=ShaderStageSwitch(PixelShader=@n15.PS, VertexShader=@n14)) {name=Preskinned, nodeColor=<rgba 0.9922 0 1 1>}) {b=true, r=true}
  n17 = Constant2Vector() {g=80}
  n18 = SphereMask(A=@n16, B=@n17) {attenuationRadius=100, hardnessPercent=0}
  NamedRerouteDeclaration(None=@n18) {name=PreskinnedMap, nodeColor=<rgba 1 0 0.0703 1>}
  n19 = SphereMask(A=@n16, B=Constant2Vector() {g=115}) {attenuationRadius=40, hardnessPercent=0}
  n20 = Constant() {r=0.3}
  n21 = Constant()
  n22 = ComponentMask(None=@n16) {r=true}
  n23 = SphereMask(A=@n22, B=Constant() {r=65}) {attenuationRadius=15, hardnessPercent=1}
  n24 = SphereMask(A=@n22, B=Constant() {r=-65}) {attenuationRadius=15, hardnessPercent=1}
  NamedRerouteDeclaration(None=Add(A=SphereMask(A=@n19, B=@n20, Hardness=@n21) {attenuationRadius=0.2, hardnessPercent=1}, B=Add(A=@n23, B=@n24))) {name=DistortionMask, nodeColor=<rgba 0.1484 0 1 1>}
  n25 = ScalarParameter() {defaultValue=0.5, group=BadBear, parameterName=BB_Limbs_PushAmount}
  n26 = WorldPosition()
  n27 = VectorParameter() {group=BadBear, parameterName=BB_BearHeartLocation, defaultValue=<rgba 0 0 107 1>}
  n28 = Constant()
  n29 = Step(Y=ComponentMask(None=MaterialFunctionCall() {materialFunction=&LocalPosition}.Local Position) {b=true}) {constX=94}
  n30 = Multiply(
   A=Multiply(A=SphereMask(A=@n26.XYZ, B=@n27.RGB, Hardness=@n28) {attenuationRadius=28, hardnessPercent=0}, B=OneMinus(None=@n29))
   B=Subtract(A=@n26.XYZ, B=@n27.RGB)
  )
  NamedRerouteDeclaration(None=LinearInterpolate(B=Multiply(A=@n25, B=@n30), Alpha=@n5)) {name=wpo, nodeColor=<rgba 0.0078 0 1 1>}
  NamedRerouteDeclaration(None=@n3) {name=SkinFX_Mask, nodeColor=<rgba 0 0.6016 1 1>}
  n31 = ComponentMask(None=NamedRerouteUsage()) {g=true}
  n32 = ScalarParameter() {defaultValue=0.02, parameterName=BB_PulseNoiseAmount}
  n33 = Time()
  n34 = ScalarParameter() {defaultValue=0.75, group=BadBear, parameterName=BB_PulseSpeed}
  n35 = OneMinus(None=Frac(None=Multiply(A=@n33, B=@n34)))
  n36 = SphereMask(A=@n18, B=Add(A=Multiply(A=@n31, B=@n32) {constB=0.04}, B=Power(Base=@n35) {constExponent=3})) {attenuationRadius=0.015, hardnessPercent=1}
  n37 = ScalarParameter() {defaultValue=0.01, parameterName=BB_PulseNoise_UnderlyingEms}
  n38 = Multiply(A=@n36, B=ScalarParameter() {defaultValue=5, parameterName=BB_PulseNoiseEms}) {constB=5}
  n39 = Multiply(A=@n31, B=ScalarParameter() {defaultValue=0.01, parameterName=BB_PulseNoiseAmount2}) {constB=0.04}
  n40 = Add(B=Multiply(A=@n33, B=@n34)) {constA=0.05}
  n41 = Power(Base=OneMinus(None=Frac(None=@n40))) {constExponent=3}
  n42 = Multiply(
   A=SphereMask(A=@n18, B=Add(A=@n39, B=@n41)) {attenuationRadius=0.04, hardnessPercent=1}
   B=ScalarParameter() {defaultValue=10, parameterName=BB_PulseNoiseEms2}
  )
  n43 = Add(A=TextureCoordinate(), B=NamedRerouteUsage())
  n44 = NamedRerouteUsage()
  n45 = LinearInterpolate(
   A=MaterialFunctionCall(In=Fresnel() {exponent=1}, Contrast=Constant() {r=0.5}) {materialFunction=&CheapContrast}.Result
   B=MaterialFunctionCall(In=Fresnel() {exponent=0.2}, Contrast=Constant() {r=0.001}) {materialFunction=&CheapContrast}.Result
   Alpha=SphereMask(A=NamedRerouteUsage(), B=Constant() {r=0.23}) {attenuationRadius=0.15, hardnessPercent=90}
  )
  n46 = ScalarParameter() {defaultValue=0.25, group=BadBear, parameterName=BB_VineEms_Min}
  n47 = ScalarParameter() {defaultValue=5, group=BadBear, parameterName=BB_VineEms_Max}
  n48 = Constant3Vector() {constant=<rgba 1 0 0 0>}
  n49 = Saturate(None=ComponentMask(None=Transform(None=@n48) {transformSourceType=TRANSFORMSOURCE_Local, transformType=TRANSFORM_Camera}) {r=true})
  n50 = Multiply(
   A=LinearInterpolate(
    A=TextureSample(UVs=@n43, Tex=@n44) {samplerSource=SSM_Wrap_WorldGroupSettings}.R
    Alpha=NamedRerouteDeclaration(None=@n45) {name=Fresnel, nodeColor=<rgba 0.4766 1 0 1>}
   ) {constB=0}
   B=LinearInterpolate(A=@n46, B=@n47, Alpha=@n49)
  )
  n51 = Add(A=TextureCoordinate(), B=NamedRerouteUsage())
  n52 = Multiply(
   A=LinearInterpolate(A=@n47, B=@n46, Alpha=@n49)
   B=LinearInterpolate(A=TextureSample(UVs=@n51, Tex=@n44) {samplerSource=SSM_Wrap_WorldGroupSettings}.G, Alpha=NamedRerouteUsage()) {constB=0}
  )
  n53 = Add(
   A=Multiply(A=@n50, B=VectorParameter() {group=BadBear, parameterName=BB_VineColor1, defaultValue=<rgba 0.8912 1 0 1>}.RGB)
   B=Multiply(A=VectorParameter() {group=BadBear, parameterName=BB_VineColor2, defaultValue=<rgba 1 0.8418 0.2812 1>}.RGB, B=@n52)
  )
  n54 = Multiply(
   A=Multiply(A=Add(A=@n38, B=@n42), B=SphereMask(A=@n16, B=@n17) {attenuationRadius=80, hardnessPercent=0})
   B=Power(Base=@n53) {constExponent=0.85}
  )
  NamedRerouteDeclaration(None=Add(A=Multiply(A=@n36, B=@n37) {constB=0.01}, B=Add(A=@n54, B=@n53))) {name=Vines, nodeColor=<rgba 1 0 0.0469 1>}
  NamedRerouteDeclaration(None=Add(A=@n50, B=@n52)) {name=VineMask, nodeColor=<rgba 0.4766 1 0 1>}
  n55 = Add(A=Divide(A=NamedRerouteUsage()) {constB=60}, B=Multiply(A=Time()) {constB=0.15})
  n56 = NamedRerouteUsage()
  n57 = MaterialFunctionCall(
   In=Abs(None=ComponentMask(None=@n56) {g=true})
   Contrast=ScalarParameter() {defaultValue=0.1, group=BadBear, parameterName=BB_TriPlanner_BlendStrength}
  ) {materialFunction=&CheapContrast}
  n58 = LinearInterpolate(
   A=TextureSample(UVs=ComponentMask(None=@n55) {b=true, g=true}) {texture=/Game/CharClasses/Outlander/Gadgets/AntiMaterialCharge/Materials/T_AntiMaterialCharge_Dissolve_01}.R
   B=TextureSample(UVs=ComponentMask(None=@n55) {b=true, r=true}) {texture=&T_AntiMaterialCharge_Dissolve_01}.R
   Alpha=Saturate(None=ComponentMask(None=@n57.Result) {g=true})
  )
  n59 = ScalarParameter() {defaultValue=0.01, parameterName=BB_VineDistortionAmount2}
  n60 = NamedRerouteUsage()
  NamedRerouteDeclaration(None=Multiply(A=ConstantBiasScale(None=@n58) {bias=-0.5, scale=2}, B=LinearInterpolate(A=@n59, Alpha=@n60) {constB=0}) {constB=0.01}) {name=Distortion_1, nodeColor=<rgba 0.7109 1 0 1>}
  n61 = Add(A=Divide(A=NamedRerouteUsage()) {constB=80}, B=Multiply(A=Time()) {constB=0.2})
  n62 = NamedRerouteUsage()
  n63 = MaterialFunctionCall(
   In=Abs(None=ComponentMask(None=@n62) {g=true})
   Contrast=ScalarParameter() {defaultValue=0.1, group=BadBear, parameterName=BB_TriPlanner_BlendStrength}
  ) {materialFunction=&CheapContrast}
  n64 = LinearInterpolate(
   A=TextureSample(UVs=ComponentMask(None=@n61) {b=true, g=true}) {texture=&T_AntiMaterialCharge_Dissolve_01}.R
   B=TextureSample(UVs=ComponentMask(None=@n61) {b=true, r=true}) {texture=&T_AntiMaterialCharge_Dissolve_01}.R
   Alpha=Saturate(None=ComponentMask(None=@n63.Result) {g=true})
  )
  n65 = ScalarParameter() {defaultValue=0.02, parameterName=BB_VineDistortionAmount1}
  n66 = NamedRerouteUsage()
  NamedRerouteDeclaration(None=Multiply(A=ConstantBiasScale(None=@n64) {bias=-0.5, scale=2}, B=LinearInterpolate(A=@n65, Alpha=@n66) {constB=0}) {constB=0.01}) {name=Distortion_2, nodeColor=<rgba 0.7109 1 0 1>}
 ]
}
