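# Niagara idle cosmetic for the purple BadBear skin: mesh drips that form on the character's skeletal mesh and fall.
# Start emitters pin wobbling, stretching drips to bones by spawn group, switching between grounded and skydiving variants via User.IsSkydiving.
# They emit location or death events, while spawner-less Drop emitters animate falling drips with inherited motion.
# Dive state flips one skydiving emitter's gravity; User.bisHandHidden culls one bone group in two emitters.
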NiagaraSystem /BRCosmetics/Athena/Effects/Particles/Skins/BadBear/NS_Skin_BadBear_Idle_purple
dependencies {
 usedDataInterfaces [
  &NiagaraDataInterfaceArrayFloat
  &NiagaraDataInterfaceArrayFloat3
  &NiagaraDataInterfaceVectorCurve
  &NiagaraDataInterfaceCurve
  &NiagaraDataInterfaceArrayInt32
 ]
 usedDynamicInputs [&FN_FrontEndFloatModifier]
 usedModules [
  &SystemState
  /Game/Effects/Niagara/Modules/Skydiving/FN_DivingAttributes/v5/FN_DivingAttributes
  &self:SystemSpawnScript.SystemScriptSource.SystemScriptGraph.NiagaraNodeAssignment_0.SetVariables_#1
 ]
}
emitters {
 Drips_Drop {
  emitterUpdateScript [&EmitterState()]
  particleSpawnScript [
   /Niagara/Modules/Spawn/Initialization/InitializeParticle(
    Lifetime=&RandomRangeFloat(Minimum=0.6, Maximum=0.8, Evaluation Type=::NewEnumerator1)
    Mesh Scale=&Multiply_VectorByFloat(Vector=<1 1 1>, Float=&RandomRangeFloat(Minimum=0.08, Maximum=0.12))
    Module.Write Color=true
    Module.Write Lifetime=true
    Module.Write Mass=true
    Module.Write Material Random=true
    Module.Write Position=false
    Module.Write Scale=true
    Module.Write SpriteRotation=false
    Module.Write SpriteSize=true
    Position=<-10 0 0>
    Sprite Size=&Vector2DFromFloat(Value=&RandomRangeFloat(Minimum=11, Maximum=14, Evaluation Type=::NewEnumerator1))
   )
   &InheritVelocity(Module.Limit Speed=false, Velocity Limit=0, Velocity Scale=<0.3 0.3 0.3>)
  ]
  particleUpdateScript [
   &ParticleState()
   &ScaleSpriteSize(Scale Factor=&Vector2DFromFloat(Value=&FloatFromCurve(FloatCurve=DataInterface() {propertyValues {Curve {keys [
          {arriveTangent=5.4603, interpMode=RCIM_Cubic, time=0, value=0.1}
          {arriveTangent=-0.1, interpMode=RCIM_Cubic, leaveTangent=-0.1, time=0.1819, value=0.9885}
          <at 1 0 RCIM_Cubic>
         ]}}})))
   &InheritSourceMovement(
    Applied Force Scale=<0 0 1>
    Applied Position Scale=&VectorFromFloat(Value=&FloatFromCurve(FloatCurve=DataInterface() {propertyValues {Curve {keys [{arriveTangent=-3.2322, interpMode=RCIM_Cubic, leaveTangent=-3.2322, tangentMode=RCTM_User, time=0, value=1}, <at 1 0 RCIM_Cubic>]}}}))
    ApplyForce=false
    ApplyPosition=true
    ApplyPositionFalloff=false
   )
   &AccelerationForce(Acceleration=&VectorFromCurve(VectorCurve=DataInterface() {propertyValues {
       XCurve {keys [<at 0 0 RCIM_Cubic>,<at 0.3 0 RCIM_Cubic>]}
       YCurve {keys [<at 0 0 RCIM_Cubic>,<at 0.3 0 RCIM_Cubic>]}
       ZCurve {keys [<at 0 0 RCIM_Cubic>, {interpMode=RCIM_Cubic, leaveTangent=-1141.1395, time=0.3, value=-600}]}
      }}))
   &ScaleColor(
    Scale Alpha=&FloatFromCurve(FloatCurve=DataInterface() {propertyValues {Curve {keys [<at 0 0>,<at 0.2 1>]}}})
    Scale RGB=&VectorFromCurve(VectorCurve=DataInterface() {propertyValues {XCurve {keys [<at 0 1>,<at 1 1>]}, YCurve {keys [<at 0 1>,<at 1 1>]}, ZCurve {keys [<at 0 1>,<at 1 1>]}}})
    ScaleA=true
    ScaleRGB=true
   )
   &CameraOffset(Camera Offset Amount=6)
   &SolveForcesAndVelocity(Clamp Velocity=false, Limit Acceleration=false)
  ]
  rendererClasses [&NiagaraMeshRendererProperties]
  renderers [{rendererClass=&NiagaraMeshRendererProperties, rendererIndex=0}]
 }
 Drips_Drop001 {
  emitterUpdateScript [&EmitterState()]
  particleSpawnScript [
   &InitializeParticle(
    Lifetime=&RandomRangeFloat(Minimum=0.6, Maximum=0.8, Evaluation Type=::NewEnumerator1)
    Mesh Scale=&Multiply_VectorByFloat(Vector=<1 1 1>, Float=&RandomRangeFloat(Minimum=0.12, Maximum=0.15))
    Module.Write Color=true
    Module.Write Lifetime=true
    Module.Write Mass=true
    Module.Write Material Random=true
    Module.Write Position=false
    Module.Write Scale=true
    Module.Write SpriteRotation=false
    Module.Write SpriteSize=true
    Position=<-10 0 0>
    Sprite Size=&Vector2DFromFloat(Value=&RandomRangeFloat(Minimum=11, Maximum=14, Evaluation Type=::NewEnumerator1))
   )
   &InheritVelocity(Module.Limit Speed=false, Velocity Limit=0, Velocity Scale=<0.1 0.1 0.01>)
  ]
  particleUpdateScript [
   &ParticleState()
   &ScaleSpriteSize(Scale Factor=&Vector2DFromFloat(Value=&FloatFromCurve(FloatCurve=DataInterface() {propertyValues {Curve {keys [
          {arriveTangent=5.4603, interpMode=RCIM_Cubic, time=0, value=0.1}
          {arriveTangent=-0.1, interpMode=RCIM_Cubic, leaveTangent=-0.1, time=0.1819, value=0.9885}
          <at 1 0 RCIM_Cubic>
         ]}}}))) {enabled=false}
   &InheritSourceMovement(
    Applied Force Scale=<0 0 1>
    Applied Position Scale=&VectorFromFloat(Value=&FloatFromCurve(FloatCurve=DataInterface() {propertyValues {Curve {keys [{arriveTangent=-3.2322, interpMode=RCIM_Cubic, leaveTangent=-3.2322, tangentMode=RCTM_User, time=0.1, value=1}, <at 1 0 RCIM_Cubic>]}}}))
    ApplyForce=false
    ApplyPosition=true
    ApplyPositionFalloff=false
   )
   &AccelerationForce(Acceleration=&VectorFromCurve(
     VectorCurve=DataInterface() {propertyValues {
       XCurve {keys [<at 0 0 RCIM_Cubic>,<at 0.3 0 RCIM_Cubic>]}
       YCurve {keys [<at 0 0 RCIM_Cubic>,<at 0.3 0 RCIM_Cubic>]}
       ZCurve {keys [<at 0 -500 RCIM_Cubic>, {interpMode=RCIM_Cubic, leaveTangent=-1141.1395, time=0.2, value=-600}]}
      }}
     Scale Curve=&VectorFromFloat(Value=&MakeCustomFloatFromBool(Bool=$User.IsSkydiving, True Float=-0.1, False Float=1))
    ))
   &ScaleColor(
    Scale Alpha=&FloatFromCurve(FloatCurve=DataInterface() {propertyValues {Curve {keys [<at 0 0>,<at 0.2 1>]}}})
    Scale RGB=&VectorFromCurve(VectorCurve=DataInterface() {propertyValues {XCurve {keys [<at 0 1>,<at 1 1>]}, YCurve {keys [<at 0 1>,<at 1 1>]}, ZCurve {keys [<at 0 1>,<at 1 1>]}}})
    ScaleA=true
    ScaleRGB=true
   ) {enabled=false}
   &CameraOffset(Camera Offset Amount=6)
   &SolveForcesAndVelocity(Clamp Velocity=false, Limit Acceleration=false)
  ]
  rendererClasses [&NiagaraMeshRendererProperties]
  renderers [{rendererClass=&NiagaraMeshRendererProperties, rendererIndex=0}]
 }
 Drips_Start {
  emitterUpdateScript [
   &EmitterState()
   &SpawnRate(
    Spawn Probability=&RandomRangeFloat(Minimum=0, Maximum=1)
    SpawnRate=&MakeCustomFloatFromBool(Bool=$User.IsSkydiving, True Float=0, False Float=3.6)
    Use Spawn Probability=true
   )
  ]
  particleSpawnScript [
   &FN_SetSpawnGroup(Number of Spawn Groups=6)
   &InitializeParticle(
    Lifetime=&RandomRangeFloat(Minimum=1.8, Maximum=2.1, Evaluation Type=::NewEnumerator1)
    Mesh Scale=&Multiply_VectorByFloat(Vector=<1 1 1>, Float=&RandomRangeFloat(Minimum=0.05, Maximum=0.1))
    Module.Write Color=true
    Module.Write Lifetime=true
    Module.Write Mass=true
    Module.Write Material Random=true
    Module.Write Position=false
    Module.Write Scale=true
    Module.Write SpriteRotation=false
    Module.Write SpriteSize=true
    Sprite Size=&Vector2DFromFloat(Value=&Multiply_Float(
      A=&RandomRangeFloat(Minimum=7, Maximum=11)
      B=&SelectFloatFromArray(Direct Array Index=$Particles.SpawnGroup, Float Selection Array=DataInterface() {propertyValues {FloatData [0.8,0.8,0.7,0.7,0.7,0.7,0.65]}})
     ))
   )
   &InitialMeshOrientation(
    Mesh Orientation Mode=::NewEnumerator2
    Rotation=&RandomRangeVector(Minimum=<0 0 -1>, Maximum=<0 0 1>)
    Use Rotation Vector=true
   )
   &CameraOffset(Camera Offset Amount=1)
  ]
  particleUpdateScript [
   &ParticleState()
   &SkeletalMeshLocation(
    Bone / Socket Index=$Particles.SpawnGroup
    Bone Sampling Mode=::NewEnumerator4
    Inherit Velocity=false
    MaskSpawnGroup=false
    Sampled Position Offset=&SelectVectorFromArray(
     Direct Array Index=$Particles.SpawnGroup
     Vector Selection Array=DataInterface() {propertyValues {FloatData [<25 -2 3>, <10 -2 3>, <-20 2.5 3>, <-5 4 3>, <0 0 3>, <-10 3 3>, <-7 10 10>]}}
    )
    Sampled Position Offset Coordinate Space:ENiagaraMeshTransforms=::NewEnumerator3
    Skeletal Mesh=$User.SkeletalMesh
   )
   &AccelerationForce(Acceleration=&VectorFromCurve(
     VectorCurve=DataInterface() {propertyValues {
       XCurve {keys [<at 0 0 RCIM_Cubic>,<at 0.3 0 RCIM_Cubic>]}
       YCurve {keys [<at 0 0 RCIM_Cubic>,<at 0.3 0 RCIM_Cubic>]}
       ZCurve {keys [<at 0 -1 RCIM_Cubic>, {interpMode=RCIM_Cubic, leaveTangent=-205.935, time=0.3, value=-20}]}
      }}
     Scale Curve=&VectorFromFloat(Value=1)
    ))
   &UpdateMeshOrientation(
    Rotation Coordinate Space=::NewEnumerator1
    Rotation Rate=&Sine(
     Normalized Angle=$Emitter.Age
     Period=&RandomRangeFloat(Minimum=0.4, Maximum=0.4)
     Scale=&FloatFromCurve(
      FloatCurve=DataInterface() {propertyValues {Curve {keys [{tangentMode=RCTM_User, time=0.2, value=0}, {tangentMode=RCTM_User, time=0.3, value=1}, <at 0.5 1>, <at 1 0>]}}}
      Scale Curve=0.4
     )
     Bias=0
    )
    Rotation Vector=&SelectVectorFromArray(
     Array Sampling Mode=::NewEnumerator0
     Direct Array Index=?
     Vector Selection Array=DataInterface() {propertyValues {FloatData [<1 0 0>, <0 1 0>]}}
     Random Seed=0
    )
   )
   &SolveForcesAndVelocity(Clamp Velocity=false, Limit Acceleration=false)
   &ScaleColor(
    Scale Alpha=&FloatFromCurve(FloatCurve=DataInterface() {propertyValues {Curve {keys [<at 0 0>,<at 0.1 1>,<at 0.9 1>,<at 1 0>]}}})
    Scale RGB=&VectorFromCurve(VectorCurve=DataInterface() {propertyValues {XCurve {keys [<at 0 1>,<at 1 1>]}, YCurve {keys [<at 0 1>,<at 1 1>]}, ZCurve {keys [<at 0 1>,<at 1 1>]}}})
    ScaleA=true
    ScaleRGB=true
   ) {enabled=false}
   &ScaleMeshSize(Scale Factor=&Multiply_VectorByFloat(
     Vector=&VectorFromCurve(VectorCurve=DataInterface() {propertyValues {
        XCurve {keys [<at 0.1 1>,<at 0.39 0.6>,<at 0.5 1>]}
        YCurve {keys [<at 0.1 1>,<at 0.39 0.6>,<at 0.5 1>]}
        ZCurve {keys [<at 0.1 1>,<at 0.39 3>,<at 0.5 1>]}
       }})
     Float=&FloatFromCurve(
      FloatCurve=DataInterface() {propertyValues {Curve {keys [
          {arriveTangent=7.5582, interpMode=RCIM_Cubic, leaveTangent=7.5582, tangentMode=RCTM_User, time=0, value=0.4}
          {arriveTangent=-0.0405, interpMode=RCIM_Cubic, leaveTangent=-0.0405, tangentMode=RCTM_User, time=0.2, value=1}
          <at 0.5 1.125>
          <at 1 0.2>
         ]}}}
      Scale Curve=4
     )
    ))
   &KillParticles(Kill Particles=&Boolean_AndOperation(
     Bool1=$User.bisHandHidden
     Bool2=&SetBoolByIntComparison(A=$Particles.SpawnGroup, B=4, Comparison Type=::NewEnumerator2)
    ))
   &GenerateDeathEvent(Event Condition:NiagaraBool=&Boolean_NotOperation(Bool=$User.bisHandHidden))
   &GenerateLocationEvent(Event Send Rate=1, Use Event Probability=true, UseEventDelay=false)
  ]
  rendererClasses [&NiagaraMeshRendererProperties]
  renderers [{rendererClass=&NiagaraMeshRendererProperties, rendererIndex=0}]
 }
 Drips_Start001 {
  emitterUpdateScript [
   &EmitterState()
   &SpawnRate(
    Spawn Probability=&RandomRangeFloat(Minimum=0, Maximum=1)
    SpawnRate=&MakeCustomFloatFromBool(Bool=$User.IsSkydiving, True Float=2.5, False Float=0)
    Use Spawn Probability=true
   )
  ]
  particleSpawnScript [
   &FN_SetSpawnGroup(Number of Spawn Groups=6)
   &InitializeParticle(
    Lifetime=&RandomRangeFloat(Minimum=1, Maximum=1.25, Evaluation Type=::NewEnumerator1)
    Mesh Scale=&Multiply_VectorByFloat(Vector=<1 1 1>, Float=&RandomRangeFloat(Minimum=0.04, Maximum=0.06))
    Module.Write Color=true
    Module.Write Lifetime=true
    Module.Write Mass=true
    Module.Write Material Random=true
    Module.Write Position=false
    Module.Write Scale=true
    Module.Write SpriteRotation=false
    Module.Write SpriteSize=true
    Sprite Size=&Vector2DFromFloat(Value=&Multiply_Float(
      A=&RandomRangeFloat(Minimum=7, Maximum=11)
      B=&SelectFloatFromArray(Direct Array Index=$Particles.SpawnGroup, Float Selection Array=DataInterface() {propertyValues {FloatData [0.8,0.8,0.7,0.7,0.7,0.7,0.65]}})
     ))
   )
   &InitialMeshOrientation(
    Mesh Axis To Orient=<0 0 -1>
    Mesh Orientation Mode=::NewEnumerator2
    Orientation Coordinate Space=::World
    Orientation Vector=<0 0 1>
    Rotation=&RandomRangeVector(Minimum=<0 0 -1>, Maximum=<0 0 1>)
    Use Rotation Vector=true
   )
   &CameraOffset(Camera Offset Amount=11)
  ]
  particleUpdateScript [
   &ParticleState()
   &SkeletalMeshLocation(
    Bone / Socket Index=$Particles.SpawnGroup
    Bone Sampling Mode=::NewEnumerator4
    Inherit Velocity=false
    MaskSpawnGroup=false
    Sampled Position Offset=&SelectVectorFromArray(
     Direct Array Index=$Particles.SpawnGroup
     Vector Selection Array=DataInterface() {propertyValues {FloatData [<25 -2 3>, <10 -2 3>, <-20 2.5 3>, <-5 4 3>, <0 0 3>, <-10 3 3>, <-7 10 10>]}}
    )
    Sampled Position Offset Coordinate Space:ENiagaraMeshTransforms=::NewEnumerator3
    Skeletal Mesh=$User.SkeletalMesh
   )
   &AccelerationForce(Acceleration=&VectorFromCurve(
     VectorCurve=DataInterface() {propertyValues {
       XCurve {keys [<at 0 0 RCIM_Cubic>,<at 0.3 0 RCIM_Cubic>]}
       YCurve {keys [<at 0 0 RCIM_Cubic>,<at 0.3 0 RCIM_Cubic>]}
       ZCurve {keys [<at 0 -1 RCIM_Cubic>, {interpMode=RCIM_Cubic, leaveTangent=-205.935, time=0.3, value=-20}]}
      }}
     Scale Curve=&VectorFromFloat(Value=1)
    ))
   &GravityForce(Gravity=&Multiply_VectorByFloat(
     Vector=<0 0 7000>
     Float=&FloatFromCurve(FloatCurve=DataInterface() {propertyValues {Curve {keys [{tangentMode=RCTM_User, time=0.45, value=0}, {tangentMode=RCTM_User, time=0.5, value=1}]}}})
    )) {enabled=false}
   &UpdateMeshOrientation(
    Rotation Coordinate Space=::NewEnumerator1
    Rotation Rate=&Sine(
     Normalized Angle=$Emitter.Age
     Period=&RandomRangeFloat(Minimum=0.4, Maximum=0.4)
     Scale=&FloatFromCurve(
      FloatCurve=DataInterface() {propertyValues {Curve {keys [{tangentMode=RCTM_User, time=0.2, value=0}, {tangentMode=RCTM_User, time=0.3, value=1}, <at 0.5 1>, <at 1 0>]}}}
      Scale Curve=0.4
     )
     Bias=0
    )
    Rotation Vector=&SelectVectorFromArray(
     Array Sampling Mode=::NewEnumerator0
     Direct Array Index=?
     Vector Selection Array=DataInterface() {propertyValues {FloatData [<1 0 0>, <0 1 0>]}}
     Random Seed=0
    )
   )
   &SolveForcesAndVelocity(Clamp Velocity=false, Limit Acceleration=false)
   &ScaleColor(
    Scale Alpha=&FloatFromCurve(FloatCurve=DataInterface() {propertyValues {Curve {keys [<at 0 0>,<at 0.1 1>,<at 0.9 1>,<at 1 0>]}}})
    Scale RGB=&VectorFromCurve(VectorCurve=DataInterface() {propertyValues {XCurve {keys [<at 0 1>,<at 1 1>]}, YCurve {keys [<at 0 1>,<at 1 1>]}, ZCurve {keys [<at 0 1>,<at 1 1>]}}})
    ScaleA=true
    ScaleRGB=true
   ) {enabled=false}
   &ScaleMeshSize(Scale Factor=&Multiply_VectorByFloat(
     Vector=&VectorFromCurve(VectorCurve=DataInterface() {propertyValues {
        XCurve {keys [<at 0.1 1>,<at 0.39 0.6>,<at 0.5 1>]}
        YCurve {keys [<at 0.1 1>,<at 0.39 0.6>,<at 0.5 1>]}
        ZCurve {keys [<at 0.1 1>,<at 0.39 2>,<at 0.5 1>]}
       }})
     Float=&FloatFromCurve(
      FloatCurve=DataInterface() {propertyValues {Curve {keys [
          {arriveTangent=7.5582, interpMode=RCIM_Cubic, leaveTangent=7.5582, tangentMode=RCTM_User, time=0, value=0.4}
          {arriveTangent=-0.0405, interpMode=RCIM_Cubic, leaveTangent=-0.0405, tangentMode=RCTM_User, time=0.2, value=1}
          <at 0.5 1.125>
          <at 1 0.2>
         ]}}}
      Scale Curve=4
     )
    ))
   &KillParticles(Kill Particles=&Boolean_AndOperation(
     Bool1=$User.bisHandHidden
     Bool2=&SetBoolByIntComparison(A=$Particles.SpawnGroup, B=4, Comparison Type=::NewEnumerator2)
    ))
   &GenerateDeathEvent() {enabled=false}
   &GenerateLocationEvent(Event Send Rate=1, Use Event Probability=true, UseEventDelay=false)
  ]
  rendererClasses [&NiagaraMeshRendererProperties]
  renderers [{rendererClass=&NiagaraMeshRendererProperties, rendererIndex=0}]
 }
 Drips_Start002 {
  emitterUpdateScript [
   &EmitterState()
   &SpawnRate(
    Spawn Probability=&RandomRangeFloat(Minimum=0, Maximum=1)
    SpawnRate=&MakeCustomFloatFromBool(Bool=$User.IsSkydiving, True Float=2.6, False Float=0)
    Use Spawn Probability=true
   )
  ]
  particleSpawnScript [
   &FN_SetSpawnGroup(Number of Spawn Groups=7)
   &InitializeParticle(
    Lifetime=&RandomRangeFloat(Minimum=1, Maximum=1.25, Evaluation Type=::NewEnumerator1)
    Mesh Scale=&Multiply_VectorByFloat(Vector=<1 1 1>, Float=&RandomRangeFloat(Minimum=0.035, Maximum=0.04))
    Module.Write Color=true
    Module.Write Lifetime=true
    Module.Write Mass=true
    Module.Write Material Random=true
    Module.Write Position=false
    Module.Write Scale=true
    Module.Write SpriteRotation=false
    Module.Write SpriteSize=true
    Sprite Size=&Vector2DFromFloat(Value=&Multiply_Float(
      A=&RandomRangeFloat(Minimum=7, Maximum=11)
      B=&SelectFloatFromArray(Direct Array Index=$Particles.SpawnGroup, Float Selection Array=DataInterface() {propertyValues {FloatData [0.8,0.8,0.7,0.7,0.7,0.7,0.65]}})
     ))
   )
   &InitialMeshOrientation(
    Mesh Axis To Orient=<0 0 -1>
    Mesh Orientation Mode=::NewEnumerator2
    Orientation Coordinate Space=::World
    Orientation Vector=<0 0 1>
    Rotation=&RandomRangeVector(Minimum=<0 0 -1>, Maximum=<0 0 1>)
    Use Rotation Vector=true
   )
   /SetVariables(Particles.VisibilityTag=&SelectIntFromArray(Direct Array Index=$Particles.SpawnGroup, Int Selection Array=DataInterface() {propertyValues {IntData [0,0,0,0,0,0,1]}})) {isSetParameters=true}
   &CameraOffset(Camera Offset Amount=25)
  ]
  particleUpdateScript [
   &ParticleState()
   &SkeletalMeshLocation(
    Bone / Socket Index=$Particles.SpawnGroup
    Bone Sampling Mode=::NewEnumerator4
    Inherit Velocity=false
    MaskSpawnGroup=false
    Sampled Position Offset=&SelectVectorFromArray(
     Direct Array Index=$Particles.SpawnGroup
     Vector Selection Array=DataInterface() {propertyValues {FloatData [<25 -2 2>, <10 -2 2>, <-20 2.5 2>, <-5 4 2>, <0 0 2>, <-10 3 2>, <-7 10 10>]}}
    )
    Sampled Position Offset Coordinate Space:ENiagaraMeshTransforms=::NewEnumerator3
    Skeletal Mesh=$User.SkeletalMesh
   )
   &AccelerationForce(Acceleration=&VectorFromCurve(
     VectorCurve=DataInterface() {propertyValues {
       XCurve {keys [<at 0 0 RCIM_Cubic>,<at 0.3 0 RCIM_Cubic>]}
       YCurve {keys [<at 0 0 RCIM_Cubic>,<at 0.3 0 RCIM_Cubic>]}
       ZCurve {keys [<at 0 -1 RCIM_Cubic>, {interpMode=RCIM_Cubic, leaveTangent=-205.935, time=0.3, value=-20}]}
      }}
     Scale Curve=&VectorFromFloat(Value=&MakeCustomFloatFromBool(Bool=$User.IsSkydiving, True Float=-1, False Float=1))
    ))
   &GravityForce(Gravity=&Multiply_VectorByFloat(
     Vector=<0 0 7000>
     Float=&FloatFromCurve(
      FloatCurve=DataInterface() {propertyValues {Curve {keys [{tangentMode=RCTM_User, time=0.45, value=0}, {tangentMode=RCTM_User, time=0.5, value=1}]}}}
      Scale Curve=&Lerp_Float(A=1, B=-1, Alpha=$System.Dive)
     )
    ))
   &UpdateMeshOrientation(
    Rotation Coordinate Space=::NewEnumerator1
    Rotation Rate=&Sine(
     Normalized Angle=$Emitter.Age
     Period=&RandomRangeFloat(Minimum=0.4, Maximum=0.4)
     Scale=&FloatFromCurve(
      FloatCurve=DataInterface() {propertyValues {Curve {keys [{tangentMode=RCTM_User, time=0.2, value=0}, {tangentMode=RCTM_User, time=0.3, value=1}, <at 0.5 1>, <at 1 0>]}}}
      Scale Curve=0.4
     )
     Bias=0
    )
    Rotation Vector=&SelectVectorFromArray(
     Array Sampling Mode=::NewEnumerator0
     Direct Array Index=?
     Vector Selection Array=DataInterface() {propertyValues {FloatData [<1 0 0>, <0 1 0>]}}
     Random Seed=0
    )
   )
   &SolveForcesAndVelocity(Clamp Velocity=false, Limit Acceleration=false)
   &ScaleColor(
    Scale Alpha=&FloatFromCurve(FloatCurve=DataInterface() {propertyValues {Curve {keys [<at 0 0>,<at 0.1 1>,<at 0.9 1>,<at 1 0>]}}})
    Scale RGB=&VectorFromCurve(VectorCurve=DataInterface() {propertyValues {XCurve {keys [<at 0 1>,<at 1 1>]}, YCurve {keys [<at 0 1>,<at 1 1>]}, ZCurve {keys [<at 0 1>,<at 1 1>]}}})
    ScaleA=true
    ScaleRGB=true
   ) {enabled=false}
   &ScaleMeshSize(Scale Factor=&Multiply_VectorByFloat(
     Vector=&VectorFromCurve(VectorCurve=DataInterface() {propertyValues {
        XCurve {keys [<at 0.1 1>,<at 0.39 0.6>,<at 0.5 1>]}
        YCurve {keys [<at 0.1 1>,<at 0.39 0.6>,<at 0.5 1>]}
        ZCurve {keys [<at 0.1 1>,<at 0.39 3>,<at 0.5 1>]}
       }})
     Float=&FloatFromCurve(
      FloatCurve=DataInterface() {propertyValues {Curve {keys [
          {arriveTangent=7.5582, interpMode=RCIM_Cubic, leaveTangent=7.5582, tangentMode=RCTM_User, time=0, value=0.4}
          {arriveTangent=-0.0405, interpMode=RCIM_Cubic, leaveTangent=-0.0405, tangentMode=RCTM_User, time=0.2, value=1}
          <at 0.5 1.125>
          <at 1 0.2>
         ]}}}
      Scale Curve=4
     )
    ))
   &GenerateDeathEvent() {enabled=false}
   &GenerateLocationEvent(Event Send Rate=1, Use Event Probability=true, UseEventDelay=false)
  ]
  rendererClasses [&NiagaraMeshRendererProperties]
  renderers [{rendererClass=&NiagaraMeshRendererProperties, rendererIndex=0}]
 }
}
summary {
 systemName=NS_Skin_BadBear_Idle_purple
 emitters [
  {emitterName=Drips_Start, rendererClasses [&NiagaraMeshRendererProperties]}
  {emitterName=Drips_Drop, rendererClasses [&NiagaraMeshRendererProperties]}
  {emitterName=Drips_Start001, rendererClasses [&NiagaraMeshRendererProperties]}
  {emitterName=Drips_Drop001, rendererClasses [&NiagaraMeshRendererProperties]}
  {emitterName=Drips_Start002, rendererClasses [&NiagaraMeshRendererProperties]}
 ]
 userVariables [
  User.bIsFrontEnd {type=NiagaraBool}
  User.bIsFrontEndPreview {type=NiagaraBool}
  User.bisHandHidden {type=NiagaraBool}
  User.IsSkydiving {type=NiagaraBool}
  User.ParaGlideLeanAlpha {type=NiagaraFloat, defaultValue=0, description=""}
  User.SkeletalMesh {
   type=NiagaraDataInterfaceSkeletalMesh
   defaultValue {
    struct=NiagaraExt_VariableValue_DataInterface
    value {dataInterface=&self:NiagaraDataInterfaceSkeletalMesh_0}
   }
  }
  User.Variant_Float {type=NiagaraFloat, defaultValue=0, description=""}
 ]
}
