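# Niagara landing effect for the BadBear skin that throws small bursts of purple mesh drops from two skeletal bones.
# Drop_Left and Drop_Right each burst a few particles on spheres around bone indices 0 and 1, sampled from the User.SkeletalMesh parameter.
# Particles get random scale and orientation, differing cone velocities per side and capped inherited horizontal velocity.
# Downward acceleration and drag pull them down while a curve shrinks them late in life.
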NiagaraSystem /BRCosmetics/Athena/Effects/Particles/Skins/BadBear/NS_Skin_BadBear_Landing_purple
dependencies {usedDataInterfaces [&NiagaraDataInterfaceCurve], usedDynamicInputs [], usedModules [&SystemState]}
emitters {
 Drop_Left {
  emitterUpdateScript [
   &EmitterState()
   &SpawnBurst_Instantaneous(Spawn Count=&RandomRangeInt(Minimum=2, Maximum=4), Spawn Time=0.025, Use Spawn Probability=true)
  ]
  particleSpawnScript [
   /Niagara/Modules/Spawn/Initialization/InitializeParticle(
    Color=<rgba 0.4065 0.2078 0.5711 1>
    Lifetime=&UniformRangedFloat(Minimum=0.3, Maximum=0.75)
    Mass=&UniformRangedFloat(Minimum=0.75, Maximum=1.75)
    Mesh Scale=&MakeVector(
     X=&UniformRangedFloat(Minimum=1.25, Maximum=2.75)
     Y=&UniformRangedFloat(Minimum=1.25, Maximum=2.75)
     Z=&UniformRangedFloat(Minimum=1, Maximum=2.75)
    )
    Module.Write Color=true
    Module.Write Lifetime=true
    Module.Write Mass=false
    Module.Write Material Random=true
    Module.Write Position=false
    Module.Write Scale=true
    Module.Write SpriteRotation=false
    Module.Write SpriteSize=false
    Position=&SimulationPosition()
    Sprite Rotation=180
   )
   &SampleSkeletalMeshSkeleton(
    Mesh:NiagaraDataInterfaceSkeletalMesh=$User.SkeletalMesh
    Sampled Position Coordinate Space:ENiagaraCoordinateSpace=::World
    Sampled Rotation Coordinate Space:ENiagaraCoordinateSpace=::World
    Sampled Velocity Coordinate Space:ENiagaraCoordinateSpace=::World
    Specific Bone Index:NiagaraInt32=&MakeIntFromFloat(Float=0)
   )
   /Niagara/Modules/Spawn/Location/SkeletalMeshSkeletonLocation(Module.Inherit Velocity=false, Module.Use Sampled Bone as Ribbon ID=true, Sampled Position Offset:Vector3f=<0 0 20>)
   /Game/Effects/Niagara/Modules/Position/FN_OffsetPositionByOrientationAxis(Offset Axis:Vector3f=<0 0 0>) {module=OffsetPositionByMeshOrientation}
   /Niagara/Modules/Spawn/Location/SphereLocation(
    MaskSpawnGroup=false
    Module.Override Randomness=false
    Sphere Radius=10
    Surface Expansion Mode:ENiagaraExpansionMode=::NewEnumerator2
    Surface Only=true
   )
   &AddVelocityInCone(
    Cone Angle=50
    Cone Axis=<5 1 1>
    Module.Use Velocity Falloff On Cone Axis=false
    Velocity Distribution Along Cone Axis=1
    Velocity Strength=&UniformRangedFloat(Minimum=100, Maximum=100)
   )
   &AddVelocity(Velocity=&UniformRangedVector(Minimum=<0 0 -25>, Maximum=<0 0 125>))
   /Niagara/Modules/Spawn/Orientation/V2/InitialMeshOrientation(
    Orientation Vector=&UniformRangedVector(Minimum=<0 0 0>, Maximum=<1 1 1>)
    Rotation=&UniformRangedVector(Minimum=<0 0 0>, Maximum=<360 360 360>)
    Use Orientation Vector=true
    Use Rotation Vector=true
   )
   &InheritVelocity(Module.Limit Speed=true, Velocity Limit=500, Velocity Scale=<1 1 0>)
   &CameraOffset(Camera Offset Amount=5)
   /SetVariables(Particles.Scale=&Multiply_VectorByFloat(
     Vector=<1 1 1>
     Float=&FloatFromCurve(
      FloatCurve=DataInterface() {propertyValues {Curve {keys [<at 0 0.6 RCIM_Constant>, {tangentMode=RCTM_User, time=0.45, value=1.1}]}}}
      CurveIndex=&RandomRangeFloat(Minimum=0, Maximum=1)
     )
    )) {isSetParameters=true}
  ]
  particleUpdateScript [
   &AccelerationForce(Acceleration=&UniformRangedVector(Minimum=<0 0 -500>, Maximum=<0 0 -1000>))
   &Drag(Drag=1.5, Use Linear Drag=true, Use Rotational Drag=false)
   &ParticleState()
   &SolveForcesAndVelocity(Clamp Velocity=false, Limit Acceleration=false)
   &ScaleMeshSize(Scale Factor=&VectorFromFloat(Value=&FloatFromCurve(FloatCurve=DataInterface() {propertyValues {Curve {keys [<at 0.5 1>,<at 1.0011 0.002>]}}}, Scale Curve=0.1)))
  ]
  rendererClasses [&NiagaraMeshRendererProperties]
  renderers [{rendererClass=&NiagaraMeshRendererProperties, rendererIndex=0}]
 }
 Drop_Right {
  emitterUpdateScript [
   &EmitterState()
   &SpawnBurst_Instantaneous(
    Spawn Count=&RandomRangeInt(Minimum=2, Maximum=4)
    Spawn Time=0.025
    Use Spawn Probability=true
    UseLoopCountLimit=false
   )
  ]
  particleSpawnScript [
   &InitializeParticle(
    Color=<rgba 0.4065 0.2078 0.5711 1>
    Lifetime=&UniformRangedFloat(Minimum=0.3, Maximum=0.75)
    Mass=&UniformRangedFloat(Minimum=0.75, Maximum=1.75)
    Mesh Scale=&MakeVector(
     X=&UniformRangedFloat(Minimum=1, Maximum=2.5)
     Y=&UniformRangedFloat(Minimum=1, Maximum=2.5)
     Z=&UniformRangedFloat(Minimum=0.9, Maximum=2.5)
    )
    Module.Write Color=true
    Module.Write Lifetime=true
    Module.Write Mass=false
    Module.Write Material Random=true
    Module.Write Position=false
    Module.Write Scale=true
    Module.Write SpriteRotation=false
    Module.Write SpriteSize=false
    Position=&SimulationPosition()
    Sprite Rotation=180
   )
   &SampleSkeletalMeshSkeleton(
    Mesh:NiagaraDataInterfaceSkeletalMesh=$User.SkeletalMesh
    Sampled Position Coordinate Space:ENiagaraCoordinateSpace=::World
    Sampled Rotation Coordinate Space:ENiagaraCoordinateSpace=::World
    Sampled Velocity Coordinate Space:ENiagaraCoordinateSpace=::World
    Specific Bone Index:NiagaraInt32=&MakeIntFromFloat(Float=1)
   )
   &SkeletalMeshSkeletonLocation(Module.Inherit Velocity=false, Module.Use Sampled Bone as Ribbon ID=true)
   &FN_OffsetPositionByOrientationAxis(Offset Axis:Vector3f=<0 0 0>) {module=OffsetPositionByMeshOrientation}
   &SphereLocation(
    Coordinate Space:ENiagaraCoordinateSpace=::Simulation
    Hemisphere Y=true
    Hemisphere Z=true
    MaskSpawnGroup=false
    Module.Override Randomness=false
    Sphere Radius=10
    Surface Expansion Mode:ENiagaraExpansionMode=::NewEnumerator2
    Surface Only=true
   )
   &AddVelocityInCone(
    Cone Angle=25
    Cone Axis=<-1 -0.25 1>
    Module.Use Velocity Falloff On Cone Axis=false
    Velocity Distribution Along Cone Axis=1
    Velocity Strength=&UniformRangedFloat(Minimum=50, Maximum=100)
   )
   &AddVelocity(Velocity=&UniformRangedVector(Minimum=<0 0 -25>, Maximum=<0 0 100>))
   &InitialMeshOrientation(
    Orientation Vector=&UniformRangedVector(Minimum=<0 0 0>, Maximum=<1 1 1>)
    Rotation=&UniformRangedVector(Minimum=<0 0 0>, Maximum=<360 360 360>)
    Use Orientation Vector=true
    Use Rotation Vector=true
   )
   &InheritVelocity(Module.Limit Speed=true, Velocity Limit=500, Velocity Scale=<1 1 0>)
   &CameraOffset(Camera Offset Amount=5)
   /SetVariables(Particles.Scale=&Multiply_VectorByFloat(
     Vector=<1 1 1>
     Float=&FloatFromCurve(
      FloatCurve=DataInterface() {propertyValues {Curve {keys [<at 0 0.6 RCIM_Constant>, {tangentMode=RCTM_User, time=0.45, value=1.1}]}}}
      CurveIndex=&RandomRangeFloat(Minimum=0, Maximum=1)
     )
    )) {isSetParameters=true}
  ]
  particleUpdateScript [
   &AccelerationForce(Acceleration=&UniformRangedVector(Minimum=<0 0 -500>, Maximum=<0 0 -1000>))
   &Drag(Drag=1.5, Use Linear Drag=true, Use Rotational Drag=false)
   &ParticleState()
   &SolveForcesAndVelocity(Clamp Velocity=false, Limit Acceleration=false)
   &ScaleMeshSize(Scale Factor=&VectorFromFloat(Value=&FloatFromCurve(FloatCurve=DataInterface() {propertyValues {Curve {keys [<at 0.5 1>,<at 1.0011 0.002>]}}}, Scale Curve=0.1)))
  ]
  rendererClasses [&NiagaraMeshRendererProperties]
  renderers [{rendererClass=&NiagaraMeshRendererProperties, rendererIndex=0}]
 }
}
summary {
 systemName=NS_Skin_BadBear_Landing_purple
 emitters [
  {emitterName=Drop_Left, rendererClasses [&NiagaraMeshRendererProperties]}
  {emitterName=Drop_Right, rendererClasses [&NiagaraMeshRendererProperties]}
 ]
 userVariables [User.SkeletalMesh {
   type=NiagaraDataInterfaceSkeletalMesh
   defaultValue {
    struct=NiagaraExt_VariableValue_DataInterface
    value {dataInterface=&self:NiagaraDataInterfaceSkeletalMesh_0}
   }
  }]
}
